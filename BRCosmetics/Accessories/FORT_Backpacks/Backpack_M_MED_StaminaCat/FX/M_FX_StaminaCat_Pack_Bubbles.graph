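# Unlit translucent Niagara mesh-particle material for bubble FX on the StaminaCat backpack, with particle color driving emissive and an adjustable camera offset.
# A panning bubble mask uses object-stabilized screen-space UVs around the particle pivot, distorted by a second texture.
# One two-channel SmoothStep thresholds both Fresnel and bubbles, with inverted Fresnel limits shifted by a bias parameter.
# The two results multiply with vertex red to form the mask.
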Material /BRCosmetics/Accessories/FORT_Backpacks/Backpack_M_MED_StaminaCat/FX/M_FX_StaminaCat_Pack_Bubbles
graph {
 properties {
  bUsedWithNiagaraMeshParticles=true
  blendMode=BLEND_Translucent
  pixelDepthOffsetMode=PDOM_Legacy
  shadingModel=MSM_Unlit
  usedShadingModels=Unlit
 }
 nodes [
  MaterialFunctionCall(Offset Amount=ScalarParameter() {parameterName=CameraOffset}) {materialFunction=/Engine/Functions/Engine_MaterialFunctions02/WorldPositionOffset/CameraOffset}
  MaterialFunctionCall(Emissive=ParticleColor().RGB) {materialFunction=&MF_FX_EmissiveAdjustments}
  n1 = Add(A=Constant2Vector() {g=0.5, r=0.6}, B=ScalarParameter() {parameterName=FresnelEdge_Bias})
  n2 = Constant2Vector() {g=0.7, r=0.5}
  n3 = MaterialFunctionCall!(
   Position=MaterialFunctionCall() {materialFunction=&ObjectPivotPoint}.Mesh Particle Pivot Location
   UVRadialScale=ScalarParameter() {parameterName=BubblesScale}
  ) {materialFunction=/Game/Weapons/WeaponSkins/MaterialFunction/ObjectStabilizedScreenSpaceUVs}
  n4 = TextureSample(UVs=Multiply(A=@n3.StabilizedUVs, B=ScalarParameter() {parameterName=DistortionMap_Size})) {texture=/Game/Effects/Fort_Effects/Textures/Power_Source/T_Power_Source_Sphere_Texture}
  n5 = ScalarParameter() {parameterName=Distortion}
  n6 = Panner(Coordinate=Add(A=@n3.StabilizedUVs, B=Multiply(A=@n4.R, B=@n5))) {speedY=0.2}
  n7 = SmoothStep(
   Min=AppendVector(A=ComponentMask(None=@n1) {r=true}, B=ComponentMask(None=@n2) {r=true})
   Max=AppendVector(A=ComponentMask(None=@n1) {g=true}, B=ComponentMask(None=@n2) {g=true})
   Value=AppendVector(
    A=NamedRerouteUsage()
    B=TextureSample(UVs=@n6) {samplerType=SAMPLERTYPE_Masks, texture=/Game/Global/Textures/Default/PointSplatter/Bubbles_M}.R
   )
  ) {constMax=0.5, constMin=0.6}
  n8 = ComponentMask(None=@n7) {r=true}
  n9 = ComponentMask(None=@n7) {g=true}
  Multiply(A=Constant() {r=1}, B=Multiply(A=Multiply(A=@n8, B=@n9), B=VertexColor().R))
  NamedRerouteDeclaration(None=MaterialFunctionCall() {materialFunction=&Fresnel_Function}.Result) {name=Fresnel, nodeColor=<rgba 1 0.4688 0 1>}
  OneMinus()
 ]
}
