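# Translucent unlit Niagara mesh-particle material drawing surface ripples around a fish tail for the StaminaCat backpack.
# It radially maps the world position behind translucency around a dynamic-parameter tail point to pan Perlin noise, sampled with explicit derivatives.
# Radial falloff and vertex red shape the noise, which is smoothstep-thresholded and sphere-masked at the tail for particle-alpha-scaled opacity.
# Particle color drives emissive, and a camera offset shifts the mesh.
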Material /BRCosmetics/Accessories/FORT_Backpacks/Backpack_M_MED_StaminaCat/FX/M_FX_StaminaCat_Pack_SurfaceRipples
graph {
 properties {
  bUsedWithNiagaraMeshParticles=true
  blendMode=BLEND_Translucent
  pixelDepthOffsetMode=PDOM_Legacy
  shadingModel=MSM_Unlit
  usedShadingModels=Unlit
 }
 nodes [
  n1 = ParticleColor()
  MaterialFunctionCall(Emissive=@n1.RGB) {materialFunction=&MF_FX_EmissiveAdjustments}
  n2 = ScalarParameter() {group=Ripples, parameterName=Ripples_Threshold_Pos}
  n3 = Add(A=@n2, B=ScalarParameter() {group=Ripples, parameterName=Ripples_Threshold_Width})
  n4 = ComponentMask(None=NamedRerouteDeclaration(None=MaterialFunctionCall() {materialFunction=&WorldPositionBehindTranslucency}.Result) {name=SurfacePos, nodeColor=<rgba 0.8047 0 1 1>}) {g=true, r=true}
  n5 = DynamicParameter() {paramNames [TailX,TailY,Param3,Param4]}
  n6 = AppendVector(A=@n5.TailX, B=@n5.TailX)
  n7 = Add(A=Divide(A=Subtract(A=@n4, B=@n6), B=ScalarParameter() {parameterName=R})) {constB=0}
  n8 = MaterialFunctionCall(Vector or UVs=@n7) {materialFunction=&VectorToRadialValue}
  n9 = VectorParameter() {parameterName=RipplesUVScale, channelNames {a=Y Speed, b=X Speed, g=Y Scale, r=X Scale}, defaultValue=<rgba 0.25 2 0 1>}
  n10 = ComponentMask(None=@n9) {g=true, r=true}
  n11 = TextureSampleParameter2D!(
   UVs=Panner(Coordinate=Multiply(A=@n8.Radial Coordinates, B=@n10), Speed=AppendVector(A=@n9.X Speed, B=@n9.X Speed)) {speedX=0.05, speedY=-0.124}
   "DDX(UVs)"=DDX(Value=@n7)
   "DDY(UVs)"=DDY(Value=@n7)
   Apply View MipBias
  ) {
   mipValueMode=TMVM_Derivative
   parameterName=RipplesNoiseTex
   samplerSource=SSM_Wrap_WorldGroupSettings
   samplerType=SAMPLERTYPE_LinearGrayscale
   texture=&T_SimplePerlin_Greyscale
  }
  n12 = Multiply(
   A=Add(A=VertexColor().R, B=ScalarParameter() {group=Ripples, parameterName=VertexColorMask_Offset})
   B=Multiply(A=OneMinus(None=@n8.Linear Distance), B=@n11.G)
  )
  Multiply(A=@n1.A, B=Multiply(A=SmoothStep(Min=@n2, Max=@n3, Value=@n12), B=NamedRerouteUsage()))
  n13 = ComponentMask(None=NamedRerouteUsage()) {g=true, r=true}
  n14 = DynamicParameter() {paramNames [TailX,TailY,Param3,Param4]}
  n15 = AppendVector(A=@n14.TailX, B=@n14.TailX)
  n16 = ScalarParameter() {desc=Hides ripples around the fish tail, parameterName=TailMaskW}
  NamedRerouteDeclaration(None=OneMinus(None=SphereMask(A=@n13, B=@n15, Radius=@n16))) {name=TailMask, nodeColor=<rgba 0.2422 0 1 1>}
  Reroute(None=MaterialFunctionCall(Offset Amount=ScalarParameter() {parameterName=CamOffset}) {materialFunction=/Engine/Functions/Engine_MaterialFunctions02/WorldPositionOffset/CameraOffset}.Camera Offset)
 ]
}
