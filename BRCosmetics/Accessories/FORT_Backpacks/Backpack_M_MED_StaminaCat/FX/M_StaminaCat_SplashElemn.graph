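# Unlit, two-sided masked Niagara material for a toon splash element of the StaminaCat backpack, reusing the Teriyaki splash texture.
# Particle alpha erodes the texture through SmoothStep, with an OutlineW-offset copy separating particle-colored fill from a dark outline.
# Their combined mask multiplies a dynamic-parameter sphere mask, whose center takes ClipSphereX for all three axes.
# Further branches compute vertex-color masking, panning-noise edge feathering, view-angle fade and toon shading.
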Material /BRCosmetics/Accessories/FORT_Backpacks/Backpack_M_MED_StaminaCat/FX/M_StaminaCat_SplashElemn
graph {
 properties {
  allowTranslucentCustomDepthWrites=true
  bTangentSpaceNormal=false
  bUsedWithNiagaraMeshParticles=true
  bUsedWithNiagaraSprites=true
  bWriteOnlyAlpha=true
  blendMode=BLEND_Masked
  opacityMaskClipValue=0.3
  pixelDepthOffsetMode=PDOM_Legacy
  shadingModel=MSM_Unlit
  twoSided=true
  usedShadingModels=Unlit
 }
 nodes [
  n1 = ParticleColor()
  n2 = TextureSample(UVs=TextureCoordinate() {uTiling=0.5}) {
   samplerSource=SSM_Clamp_WorldGroupSettings
   samplerType=SAMPLERTYPE_LinearColor
   texture=/BRCosmetics/Accessories/FORT_Backpacks/Backpack_M_MED_Teriyaki_Fish_Toon/FX/T_FX_TeriyakiToon_Splash
  }
  n3 = ScalarParameter() {parameterName=OutlineW}
  n4 = SmoothStep(Min=@n1.A, Max=Add(A=@n1.A, B=ScalarParameter() {parameterName=w}), Value=AppendVector(A=@n2.G, B=Add(A=@n2.G, B=@n3)))
  n5 = ComponentMask(None=@n4) {r=true}
  LinearInterpolate(A=Constant(), B=@n1.RGB, Alpha=@n5) {constA=1}
  n6 = Constant() {r=1}
  n7 = ComponentMask(None=TextureCoordinate()) {g=true}
  n8 = Saturate(None=DotProduct(A=Constant3Vector() {constant=<rgba 0 0 1 0>}, B=CameraVectorWS()))
  n9 = ScalarParameter() {group=AngleFade, parameterName=AngleFade_Min}
  n10 = ScalarParameter() {group=AngleFade, parameterName=AngleFade_Max}
  n11 = Constant()
  n12 = Constant() {r=1}
  LinearInterpolate(
   A=@n6
   B=MaterialFunctionCall(Input=OneMinus(None=@n7), Input Low=Constant() {r=0.3}, Input High=Constant() {r=0.4}, Target Low=Constant(), Target High=Constant() {r=1}) {materialFunction=&RemapValueRange}.Result
   Alpha=Saturate(None=MaterialFunctionCall(Input=@n8, Input Low=@n9, Input High=@n10, Target Low=@n11, Target High=@n12) {materialFunction=&RemapValueRange}.Result)
  )
  n13 = MakeMaterialAttributes!(
   BaseColor
   Metallic
   Specular
   Roughness
   Anisotropy
   EmissiveColor=@n1.RGB
   Opacity
   OpacityMask
   Normal=Constant3Vector() {constant=<rgba 0 0 1 0>}
   Tangent
   WorldPositionOffset
   SubsurfaceColor
   ClearCoat
   ClearCoatRoughness
   AmbientOcclusion
   Refraction
   CustomizedUVs_0
   CustomizedUVs_1
   CustomizedUVs_2
   CustomizedUVs_3
   CustomizedUVs_4
   CustomizedUVs_5
   CustomizedUVs_6
   CustomizedUVs_7
   PixelDepthOffset
   ShadingModel
   Displacement
  )
  MaterialFunctionCall(Emissive=BreakMaterialAttributes!(Attr=@n13).EmissiveColor) {materialFunction=&MF_FX_EmissiveAdjustments}
  MaterialFunctionCall!(SkinMask, In=@n13) {materialFunction=/Game/Packages/Fortress_SharedMaterials/Anime_Master_Material/MF_Toonify}
  n14 = ComponentMask(None=@n4) {g=true}
  n15 = DynamicParameter() {paramNames [ClipSphereX, ClipSphereY, ClipSphereZ, ClipSphereRadius]}
  n16 = Multiply(
   A=Saturate(None=Add(A=@n5, B=@n14))
   B=SphereMask(
    A=MaterialFunctionCall(R=@n15.ClipSphereX, G=@n15.ClipSphereX, B=@n15.ClipSphereX) {materialFunction=&AppendMany}.RGB
    B=WorldPosition().XYZ
    Radius=@n15.ClipSphereRadius
   )
  )
  Multiply(A=@n16, B=VertexColor().R)
  n17 = TextureCoordinate() {uTiling=2, vTiling=1.25}
  n18 = Add(
   A=TextureSample(UVs=Panner(Coordinate=@n17) {speedY=0.2}) {texture=/Game/Characters/Player/Male/Medium/Bodies/M_MED_Viper/Textures/Glitch/T_SoftNoise}.G
   B=ScalarParameter() {parameterName=Edge_Feather_Pos}
  )
  n19 = ScalarParameter() {parameterName=Edge_Feather_W}
  n20 = TextureCoordinate() {uTiling=2}
  Multiply(A=@n16, B=SmoothStep(Min=@n18, Max=Add(A=@n18, B=@n19), Value=ComponentMask(None=@n20) {g=true}))
  NamedRerouteDeclaration(None=@n6) {name=SurfaceBlend, nodeColor=<rgba 0 0.4609 1 1>}
 ]
}
