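# Niagara jump effect for the male StaminaCat backpack cosmetic, showing water splashes at the sides of a bowl.
# System scripts sample the skeletal mesh and set variables such as the bowl center.
# Only SideSplashes001 is enabled: two timed bursts spawn particles mirrored by an IsLHS flag, which shrink in height while alpha dips and recovers.
# Bowl center feeds dynamic material parameters; the RunReactive droplet and older SideSplashes emitters are disabled, as are speed-driven spawn rates.
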NiagaraSystem /BRCosmetics/Accessories/FORT_Backpacks/Backpack_M_MED_StaminaCat/FX/NS_Backpack_StaminaCatMale_Jump
dependencies {
 usedDataInterfaces [&NiagaraDataInterfaceCurve, &NiagaraDataInterfaceVector2DCurve, &NiagaraDataInterfaceColorCurve]
 usedDynamicInputs [
  /Niagara/DynamicInputs/TypeConversions/Vector2DLength
  &VectorLength
  &SetBoolByFloatComparison
  &Abs_Float
 ]
 usedModules [
  &self:SystemSpawnScript.SystemScriptSource.SystemScriptGraph.NiagaraNodeAssignment_7.SetVariables_#1
  &SampleSkeletalMesh
  &SystemState
  &self:SystemSpawnScript.SystemScriptSource.SystemScriptGraph.NiagaraNodeAssignment_2.SetVariables_#2
  &self:SystemSpawnScript.SystemScriptSource.SystemScriptGraph.NiagaraNodeAssignment_4.SetVariables_#3
  &self:SystemSpawnScript.SystemScriptSource.SystemScriptGraph.NiagaraNodeAssignment_0.SetVariables_#4
  &self:SystemSpawnScript.SystemScriptSource.SystemScriptGraph.NiagaraNodeAssignment_1.SetVariables_#5
  &self:SystemSpawnScript.SystemScriptSource.SystemScriptGraph.NiagaraNodeAssignment_6.SetVariables_#6
 ]
}
emitters {
 RunReactive {
  bEnabled=false
  emitterUpdateScript [
   &EmitterState(Life Cycle Mode=::NewEnumerator1, Loop Behavior=::NewEnumerator1, Loop Duration=0.5, UseLoopDelay=false)
   &SpawnBurst_Instantaneous(Spawn Count=&RandomRangeInt(Minimum=3, Maximum=3), Use Spawn Probability=true, UseLoopCountLimit=false)
  ]
  particleSpawnScript [
   &InitializeParticle(
    Color=<rgba 0.0529 0.2604 0.1325 1>
    Lifetime Max=1
    Lifetime Min=0.2
    Lifetime Mode=::NewEnumerator1
    Mass Mode=::NewEnumerator3
    Mesh Scale=&Multiply_VectorByFloat(Vector=<1 1 1>, Float=&RandomRangeFloat(Minimum=0.5, Maximum=1.2))
    Mesh Scale Mode=::NewEnumerator4
    Sprite Size Mode=::NewEnumerator1
    Uniform Sprite Size Max=20
    Uniform Sprite Size Min=10
    UsePositionOffset=false
    Write Material Random=true
   )
   &CylinderLocation(
    Cylinder Height=1
    Cylinder Origin:NiagaraPosition=$System.WaterLevelPos
    Cylinder Radius=2
    MaskSpawnGroup=false
    Override Seed=false
    UseOwnerScale=true
   )
   &AddVelocityInCone(
    Cone Angle=35
    Cone Axis=<0 0 1>
    Module.Use Velocity Falloff On Cone Axis=false
    Velocity Strength=&RandomRangeFloat(Minimum=100, Maximum=400)
   )
   &SubUVAnimation(End Frame=3, SubUV Animation Mode:ENiagara_SubUVLookupMode=::NewEnumerator2)
  ]
  particleUpdateScript [
   &ParticleState()
   &GravityForce()
   &Drag(Use Linear Drag=true, Use Rotational Drag=false)
   &ScaleColor(
    Scale Alpha=&FloatFromCurve(FloatCurve=DataInterface() {propertyValues {Curve {keys [<at -0.0046 1.0113 RCIM_Cubic>,<at 1.0011 0.002 RCIM_Cubic>]}}})
    ScaleA=true
    ScaleRGB=true
   )
   &InheritSourceMovement(
    Applied Force Scale=&VectorFromFloat(Value=&FloatFromCurve(FloatCurve=DataInterface() {propertyValues {Curve {keys [{arriveTangent=0.4365, interpMode=RCIM_Cubic, leaveTangent=0.4365, tangentMode=RCTM_User, time=0, value=1}, <at 1 0 RCIM_Cubic>]}}}))
    Applied Position Scale=&VectorFromFloat(Value=&FloatFromCurve(FloatCurve=DataInterface() {propertyValues {Curve {keys [<at 0 1 RCIM_Cubic>,<at 0.9 0 RCIM_Cubic>]}}}))
    ApplyForce=true
    ApplyForceFalloff=false
    ApplyPosition=true
    ApplyPositionFalloff=false
    Module.LimitForceSpeed=false
   )
   &SolveForcesAndVelocity(Clamp Velocity=false, Limit Acceleration=false)
   &CameraOffset()
   &ScaleMeshSize(Scale Factor=&VectorFromFloat(Value=&FloatFromCurve(FloatCurve=DataInterface() {propertyValues {Curve {keys [
          <at 0 1 RCIM_Cubic>
          {arriveTangent=-0.6786, interpMode=RCIM_Cubic, leaveTangent=-0.6786, time=0.4, value=1.3}
          {arriveTangent=-2.1667, interpMode=RCIM_Cubic, leaveTangent=-2.1667, time=0.7, value=0.525}
          <at 1 0 RCIM_Cubic>
         ]}}})))
   &DynamicMaterialParameters(
    Index 0 Param 1=0.5
    Module.Param0WriteEnabled=true
    Module.Param1WriteEnabled=true
    Module.Param2WriteEnabled=true
    Module.Param3WriteEnabled=true
   )
   &ScaleMeshSizeBySpeed(
    Max Scale Factor=<5 1 1>
    Sample Scale Factor By Curve=true
    Scale Factor Curve=DataInterface() {propertyValues {Curve {keys [<at 0 0 RCIM_Cubic>, {arriveTangent=1, interpMode=RCIM_Cubic, leaveTangent=1, time=0.6, value=0.175}, <at 1 1 RCIM_Cubic>]}}}
    Speed Threshold=100
   ) {enabled=false}
   &ScaleMeshSize(Scale Factor=&MakeVector(
     X=&FloatFromCurve(FloatCurve=DataInterface() {propertyValues {Curve {keys [
          {arriveTangent=-3.875, interpMode=RCIM_Cubic, time=0, value=1}
          {arriveTangent=-0.625, interpMode=RCIM_Cubic, leaveTangent=-0.625, time=0.2, value=3}
          <at 1 0.375 RCIM_Cubic>
         ]}}})
     Y=1
     Z=1
    )) {module=ScaleMeshSize001}
  ]
  rendererClasses [&NiagaraMeshRendererProperties, &NiagaraSpriteRendererProperties]
  renderers [
   {rendererClass=&NiagaraMeshRendererProperties, rendererIndex=0}
   {rendererClass=&NiagaraSpriteRendererProperties, rendererIndex=1}
  ]
 }
 SideSplashes {
  bEnabled=false
  emitterUpdateScript [
   &EmitterState(Life Cycle Mode=::NewEnumerator1, Loop Duration=0.2, UseLoopDelay=false)
   &SpawnRate(SpawnRate=5, Use Spawn Probability=false) {enabled=false, module=SpawnRate001}
   &SpawnBurst_Instantaneous(Use Spawn Probability=false, UseLoopCountLimit=false)
   &SpawnBurst_Instantaneous(Spawn Time=0.05, Use Spawn Probability=false, UseLoopCountLimit=false) {module=SpawnBurst_Instantaneous001}
  ]
  particleSpawnScript [
   &InitializeParticle(
    Color=<rgba 0.0617 0.6562 0.4839 1>
    Lifetime Max=0.85
    Lifetime Min=0.45
    Lifetime Mode=::NewEnumerator1
    Sprite Size=&Multiply_Vector2D_ByFloat(Float=1.5, Vector2D=<6 7>)
    Sprite Size Mode=::NewEnumerator4
    UsePositionOffset=true
    Write Material Random=true
   )
   /SetVariables(Particles.IsLHS:NiagaraBool=&SetBoolByIntComparison(A=$Particles.UniqueID, B=0, Comparison Type=::NewEnumerator0)) {isSetParameters=true, module=SetVariables_#7}
   /SetVariables(Particles.PivotOffset:Vector2f=&MakeCustomVector2DFromBool(Bool=$Particles.IsLHS, True Vector2D=<0.7 1>, False Vector2D=<0.2 1>)) {isSetParameters=true}
   &SubUVAnimation(End Frame=1, SubUV Animation Mode:ENiagara_SubUVLookupMode=::NewEnumerator2)
  ]
  particleUpdateScript [
   &ParticleState()
   /SetVariables() {isSetParameters=true, module=SetVariables_#8}
   /SetVariables(Particles.SpriteAlignment=<0 0 1>) {isSetParameters=true, module=SetVariables_#9}
   &DynamicMaterialParameters(
    Index 0 Param 1=&MakeFloatFromVector(Vector=$System.BowlCenter, Channel=::NewEnumerator0)
    Index 0 Param 2=&MakeFloatFromVector(Vector=$System.BowlCenter, Channel=::NewEnumerator1)
    Index 0 Param 3=&MakeFloatFromVector(Vector=$System.BowlCenter, Channel=::NewEnumerator2)
    Index 0 Param 4=7.8972
    Module.Param0WriteEnabled=true
    Module.Param1WriteEnabled=true
    Module.Param2WriteEnabled=true
    Module.Param3WriteEnabled=true
   )
   &ScaleSpriteSize(
    Non-Uniform Curve Sprite Scale=DataInterface() {propertyValues {
      XCurve {keys [
        {arriveTangent=2.8681, interpMode=RCIM_Cubic, leaveTangent=2.8681, tangentMode=RCTM_User, time=0, value=1}
        {arriveTangent=1.25, interpMode=RCIM_Cubic, leaveTangent=1.25, time=0.4, value=1.75}
        <at 1 2.25 RCIM_Cubic>
       ]}
      YCurve {keys [
        <at 0 1 RCIM_Cubic>
        {arriveTangent=0.3, interpMode=RCIM_Cubic, leaveTangent=0.3, time=0.1, value=2.425}
        {arriveTangent=-1.5833, interpMode=RCIM_Cubic, leaveTangent=-1.5833, time=0.5, value=1.15}
        <at 1 1 RCIM_Cubic>
       ]}
     }}
    Scale Sprite Size Mode=::NewEnumerator3
   )
   &ScaleSpriteSize(
    Scale Factor=&MakeVector2D(X=&MakeCustomFloatFromBool(Bool=$Particles.IsLHS, True Float=-1, False Float=1), Y=1)
    Scale Sprite Size Mode=::NewEnumerator2
   ) {module=ScaleSpriteSize001}
   &CameraOffset(Camera Offset Amount=&FloatFromCurve(
     FloatCurve=DataInterface() {propertyValues {Curve {keys [{interpMode=RCIM_Cubic, tangentMode=RCTM_User, time=0, value=0}, {interpMode=RCIM_Cubic, tangentMode=RCTM_User, time=1, value=1}]}}}
     Scale Curve=0
    )) {enabled=false}
   &Color(Color=&MakeLinearColorFromVectorAndFloat(
     "Vector (RGB)"=<1 1 1>
     "Float (Alpha)"=&Lerp_Float(
      A=0.3
      B=1
      Alpha=&FloatFromCurve(FloatCurve=DataInterface() {propertyValues {Curve {keys [
           {arriveTangent=0.1629, interpMode=RCIM_Cubic, time=0, value=0.975}
           {arriveTangent=0.0278, interpMode=RCIM_Cubic, leaveTangent=0.0278, time=0.2, value=0.025}
           <at 0.9 1 RCIM_Cubic>
          ]}}})
     )
    )) {enabled=false}
   &ScaleColor(
    Scale Alpha=&Lerp_Float(
     A=0.3
     B=1
     Alpha=&FloatFromCurve(FloatCurve=DataInterface() {propertyValues {Curve {keys [
          {arriveTangent=0.1629, interpMode=RCIM_Cubic, time=0, value=0.975}
          {arriveTangent=0.0278, interpMode=RCIM_Cubic, leaveTangent=0.0278, time=0.2, value=0.025}
          <at 0.9 1 RCIM_Cubic>
         ]}}})
    )
    ScaleA=true
    ScaleRGB=false
   )
   &CameraOffset(Camera Offset Amount=2) {module=CameraOffset001}
  ]
  rendererClasses [&NiagaraSpriteRendererProperties]
  renderers [
   {rendererClass=&NiagaraSpriteRendererProperties}
   {rendererClass=&NiagaraSpriteRendererProperties, rendererIndex=1}
  ]
 }
 SideSplashes001 {
  emitterUpdateScript [
   &EmitterState(Life Cycle Mode=::NewEnumerator1, Loop Behavior=::NewEnumerator1, Loop Duration=0.5, UseLoopDelay=false)
   &SpawnRate(
    SpawnRate=&MakeCustomFloatFromBool(
     Bool=$System.JumpingUp
     True Float=0
     False Float=&RemapRange(Input Value To Remap=$System.MovingForward, Input Min=0, Input Max=1, Output Min=0, Output Max=25)
    )
    Use Spawn Probability=false
   ) {enabled=false}
   &SpawnRate(SpawnRate=3, Use Spawn Probability=false) {enabled=false, module=SpawnRate001}
   &SpawnBurst_Instantaneous(Use Spawn Probability=false, UseLoopCountLimit=false)
   &SpawnBurst_Instantaneous(Spawn Time=0.1, Use Spawn Probability=false, UseLoopCountLimit=false) {module=SpawnBurst_Instantaneous001}
  ]
  particleSpawnScript [
   &InitializeParticle(
    Color=<rgba 0.0617 0.6562 0.4839 1>
    Lifetime Max=0.9
    Lifetime Min=0.7
    Lifetime Mode=::NewEnumerator1
    Mesh Scale=<1 2 1.5>
    Mesh Scale Mode=::NewEnumerator4
    Position Offset=<0 0 -1.9169>
    Sprite Size=&Multiply_Vector2D_ByFloat(Float=1.4, Vector2D=<6 7>)
    Sprite Size Mode=::NewEnumerator4
    UsePositionOffset=true
    Write Material Random=true
   )
   /SetVariables(Particles.IsLHS:NiagaraBool=&self:ScratchDynamicInput_04()) {isSetParameters=true, module=SetVariables_#7}
   &SubUVAnimation(End Frame=1, SubUV Animation Mode:ENiagara_SubUVLookupMode=::NewEnumerator2) {enabled=false}
   /SetVariables(Particles.Scale=&MakeVector(
     X=1
     Y=&Multiply_Float(A=2.25, B=&MakeCustomFloatFromBool(Bool=$Particles.IsLHS, True Float=1, False Float=-1))
     Z=1.6647
    )) {isSetParameters=true}
   &InitialMeshOrientation(Mesh Axis To Orient=<0 1 0>, Mesh Orientation Mode=::NewEnumerator2, Orientation Vector=<0 1 0>, Use Rotation Vector=false)
  ]
  particleUpdateScript [
   &ParticleState()
   /SetVariables() {enabled=false, isSetParameters=true, module=SetVariables_#8}
   /SetVariables(Particles.SpriteAlignment=<0 0 1>) {enabled=false, isSetParameters=true, module=SetVariables_#9}
   &DynamicMaterialParameters(
    Index 0 Param 1=&MakeFloatFromVector(Vector=$System.BowlCenter, Channel=::NewEnumerator0)
    Index 0 Param 2=&MakeFloatFromVector(Vector=$System.BowlCenter, Channel=::NewEnumerator1)
    Index 0 Param 3=&MakeFloatFromVector(Vector=$System.BowlCenter, Channel=::NewEnumerator2)
    Index 0 Param 4=8.5
    Module.Param0WriteEnabled=true
    Module.Param1WriteEnabled=true
    Module.Param2WriteEnabled=true
    Module.Param3WriteEnabled=true
   )
   &CameraOffset(Camera Offset Amount=3)
   &Color(Color=&ColorFromCurve(ColorCurve=DataInterface() {propertyValues {
       AlphaCurve {keys [<at 0 1 RCIM_Cubic>]}
       BlueCurve {keys [<at 0.0658 0.0848 RCIM_Cubic>]}
       GreenCurve {keys [<at 0.0658 0.1667 RCIM_Cubic>]}
       RedCurve {keys [<at 0.0658 0.0339 RCIM_Cubic>]}
      }}))
   &ScaleColor(
    Color Value To Scale=$Particles.Color
    Scale Alpha=&Lerp_Float(
     A=0.3
     B=1
     Alpha=&FloatFromCurve(FloatCurve=DataInterface() {propertyValues {Curve {keys [
          {arriveTangent=0.1629, interpMode=RCIM_Cubic, time=0, value=0.975}
          {arriveTangent=0.0278, interpMode=RCIM_Cubic, leaveTangent=0.0278, time=0.2, value=0.025}
          <at 0.9 1 RCIM_Cubic>
         ]}}})
    )
    ScaleA=true
    ScaleRGB=true
   )
   &ScaleMeshSize(Scale Factor=&MakeVector(
     X=1
     Y=1
     Z=&FloatFromCurve(
      FloatCurve=DataInterface() {propertyValues {Curve {keys [
          {arriveTangent=-0.7812, interpMode=RCIM_Cubic, time=0, value=1}
          {arriveTangent=-1, interpMode=RCIM_Cubic, leaveTangent=-1, time=0.6, value=0.375}
          <at 1 0 RCIM_Cubic>
         ]}}}
      Scale Curve=1.5
     )
    ))
   &UpdateMeshOrientation(Facing Interpolation=false, Orientation Method=::NewEnumerator1, Override Quat=false, TransformFacing=false, TransformUp=false) {enabled=false}
  ]
  rendererClasses [&NiagaraSpriteRendererProperties, &NiagaraMeshRendererProperties]
  renderers [
   {rendererClass=&NiagaraSpriteRendererProperties}
   {rendererClass=&NiagaraMeshRendererProperties, rendererIndex=1}
   {rendererClass=&NiagaraSpriteRendererProperties, rendererIndex=2}
  ]
 }
}
summary {
 systemName=NS_Backpack_StaminaCatMale_Jump
 emitters [
  {
   bEnabled=false
   emitterName=RunReactive
   rendererClasses [&NiagaraMeshRendererProperties, &NiagaraSpriteRendererProperties]
  }
  {bEnabled=false, emitterName=SideSplashes, rendererClasses [&NiagaraSpriteRendererProperties]}
  {
   emitterName=SideSplashes001
   rendererClasses [&NiagaraSpriteRendererProperties, &NiagaraMeshRendererProperties]
  }
 ]
 userVariables [
  User.IsSkydiving {type=NiagaraBool}
  User.SkeletalMesh {
   type=NiagaraDataInterfaceSkeletalMesh
   defaultValue {
    struct=NiagaraExt_VariableValue_DataInterface
    value {dataInterface=&self:NiagaraDataInterfaceSkeletalMesh_0}
   }
  }
  User.WaterHighlightColor {type=LinearColor, description="", defaultValue {struct=LinearColor, value=<rgba 0.1186 0.4844 0.3089 1>}}
 ]
}
