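# Additive unlit anime-style glass material for the StaminaCat backpack accessory, usable on skeletal meshes.
# A sky light environment reflection is tinted by two Fresnel-blended glass colors, then a rounded, tighter Fresnel switches to a bright edge color for a hard cel-style rim.
# The saturated result feeds base color and emissive attributes, which branch into MF_TechArtOverrides_accessories and into MF_Toonify with a re-saturated emissive.
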Material /BRCosmetics/Accessories/FORT_Backpacks/Backpack_M_MED_StaminaCat/Materials/M_AnimeGlass
graph {
 properties {
  bUseMaterialAttributes=true
  bUsedWithMeshDeformer=true
  bUsedWithSkeletalMesh=true
  blendMode=BLEND_Additive
  pixelDepthOffsetMode=PDOM_Legacy
  shadingModel=MSM_Unlit
  usedShadingModels=Unlit
 }
 nodes [
  Constant()
  n1 = SkyLightEnvMapSample(Direction=ReflectionVectorWS(), Roughness=Constant())
  n2 = Fresnel(ExponentIn=ScalarParameter() {defaultValue=2, parameterName=GlassinnerEXP}) {baseReflectFraction=0}
  n3 = LinearInterpolate(
   A=VectorParameter() {parameterName=GlassColor, defaultValue=<rgba 0.1936 0.3264 0.3802 1>}.RGB
   B=VectorParameter() {parameterName=GlassColorb, defaultValue=<rgba 0.1936 0.3264 0.3802 1>}.RGB
   Alpha=@n2
  )
  n4 = Fresnel(ExponentIn=ScalarParameter() {defaultValue=8, parameterName=GlassEdgeEXP}) {baseReflectFraction=0}
  n5 = Saturate(None=LinearInterpolate(
    A=Multiply(A=@n1, B=@n3)
    B=VectorParameter() {parameterName=GlassColorEDGE, defaultValue=<rgba 50 41.0803 34.3924 1>}.RGB
    Alpha=Round(None=@n4)
   ))
  n6 = ScalarParameter() {defaultValue=1, parameterName=outline}
  n7 = SetMaterialAttributes!(MaterialAttributes, Base Color=@n5, Metallic=@n6, Specular=@n6, Roughness=@n6, Opacity=@n6, Emissive Color=@n5) {attributeSetTypes [
    #1
    #2
    #3
    #4
    #5
    #6
   ]}
  MaterialFunctionCall(In=SetMaterialAttributes!(MaterialAttributes=@n7, Shading Model=ShadingModel()) {attributeSetTypes [#7]}) {materialFunction=&MF_TechArtOverrides_accessories}
  Round(None=@n2)
  n8 = GetMaterialAttributes(None=MaterialFunctionCall!(SkinMask, In=@n7) {materialFunction=/Game/Packages/Fortress_SharedMaterials/Anime_Master_Material/MF_Toonify}.Result) {attributeGetTypes [#6]}
  SetMaterialAttributes!(MaterialAttributes=@n8.MaterialAttributes, Emissive Color=Saturate(None=@n8.Emissive Color)) {attributeSetTypes [#6]}
 ]
}
