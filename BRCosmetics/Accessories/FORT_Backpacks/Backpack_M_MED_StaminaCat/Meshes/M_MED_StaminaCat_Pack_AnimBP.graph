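# Animation Blueprint for the StaminaCat backpack cosmetic, animating its head fins and adjusting the pack during skydiving.
# A two-state machine switches between loop and action fin animations near clip end, using a bool randomized by the RandomTrigger notify.
# Its pose feeds additive bone offsets around a half-alpha RigidBody node with overridden gravity.
# Skydiving tilts the pack before simulation; skydive floating rotates and scales the dyn_water_base bone afterward, alongside fixed corrective offsets.
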AnimBlueprint /BRCosmetics/Accessories/FORT_Backpacks/Backpack_M_MED_StaminaCat/Meshes/M_MED_StaminaCat_Pack_AnimBP
target_skeleton=&/M_MED_StaminaCat_Pack_Skeleton
graphs [
 AnimGraph {nodes [
   Input Pose()
   AnimGraphNode_StateMachine_0 = Random State Machine() {is=title AnimGraphNode_StateMachine}
   n1 = "Transform (Modify) Bone - Bone: pack"!(
    ComponentPose=Local To Component(LocalPose=@AnimGraphNode_StateMachine_0)
    bAlphaBoolEnabled:bool=Get bIsSkyDiving()
    Translation="0,0,0"
    Rotation="-0.000,0.000,20.000"
    Scale
   ) -> (Pose) {
    alphaInputType=Bool
    rotationMode=BMM_Additive
    rotationSpace=BCS_BoneSpace
    scaleMode=BMM_Additive
    translationMode=BMM_Additive
    translationSpace=BCS_BoneSpace
    alphaBoolBlend {blendOption=Linear}
    boneToModify {boneName=pack}
    rotation {roll=20}
   }
   n2 = RigidBody(ComponentPose=@n1, Alpha=0.5) {
    bOverrideWorldGravity=true
    alphaBoolBlend {blendOption=Linear}
    componentLinearAccScale=<0.7 0.7 0.7>
    componentLinearVelScale=<0.7 0.7 0.7>
    overrideWorldGravity {z=-980}
   }
   n3 = "Transform (Modify) Bone - Bone: dyn_water_base"!(ComponentPose=@n2, bAlphaBoolEnabled:bool=Get bIsSkydiveFloating(), Translation="0,0,0", Rotation="0,0,-62", Scale="0.9,0.9,0.9") -> (Pose) {
    alphaInputType=Bool
    rotationMode=BMM_Additive
    rotationSpace=BCS_BoneSpace
    scaleMode=BMM_Additive
    scaleSpace=BCS_BoneSpace
    translationMode=BMM_Additive
    translationSpace=BCS_BoneSpace
    alphaBoolBlend {blendOption=Linear}
    boneToModify {boneName=dyn_water_base}
    rotation {roll=-62}
    scale=<0.9 0.9 0.9>
   }
   n4 = "Transform (Modify) Bone - Bone: dyn_water_base"!(ComponentPose=@n3, Alpha:float=1, Translation="0,0,0", Rotation="0,0,-13", Scale="1,1,1") -> (Pose) {
    rotationMode=BMM_Additive
    rotationSpace=BCS_BoneSpace
    scaleMode=BMM_Additive
    scaleSpace=BCS_BoneSpace
    translationMode=BMM_Additive
    translationSpace=BCS_BoneSpace
    alphaBoolBlend {blendOption=Linear}
    boneToModify {boneName=dyn_water_base}
    rotation {roll=-13}
   }
   n5 = "Transform (Modify) Bone - Bone: pack"!(ComponentPose=@n4, Alpha:float, Translation="0,1,0", Rotation="0,0,0", Scale) -> (Pose) {
    translationMode=BMM_Additive
    translationSpace=BCS_BoneSpace
    alphaBoolBlend {blendOption=Linear}
    boneToModify {boneName=pack}
    translation {y=1}
   }
   Output Pose(Result=Component To Local(ComponentPose=@n5)) {layerGroup=None}
   Sequence Player 'StaminaCat_HeadFins_Action'() {sequence=/BRCosmetics/Accessories/FORT_Backpacks/Backpack_M_MED_StaminaCat/Animation/StaminaCat_HeadFins_Action}
   Sequence Player 'StaminaCat_HeadFins_Loop'() {sequence=/BRCosmetics/Accessories/FORT_Backpacks/Backpack_M_MED_StaminaCat/Animation/StaminaCat_HeadFins_Loop}
  ]}
 Random State Machine {
  entry=SmallMovement
  path=AnimGraph.AnimGraphNode_StateMachine_0.Random State Machine
  ids=id
  nodes [
   AnimStateNode_0 = SmallMovement() {is=title AnimStateNode}
   AnimStateNode_1 = BigMovement() {is=title AnimStateNode}
   AnimStateTransitionNode_4 = AnimStateTransitionNode() {crossfadeDuration=0.1}
   AnimStateTransitionNode_5 = AnimStateTransitionNode() {crossfadeDuration=0.1}
  ]
  links [
   {
    dst=SmallMovement
    rule=AnimGraph.AnimGraphNode_StateMachine_0.Random State Machine.AnimStateTransitionNode_5.Transition
    src=BigMovement
    via=AnimStateTransitionNode_5
   }
   {
    dst=BigMovement
    rule=AnimGraph.AnimGraphNode_StateMachine_0.Random State Machine.AnimStateTransitionNode_4.Transition
    src=SmallMovement
    via=AnimStateTransitionNode_4
   }
  ]
 }
 SmallMovement {
  path=AnimGraph.AnimGraphNode_StateMachine_0.Random State Machine.AnimStateNode_0.SmallMovement
  nodes [Output Animation Pose(Result=Sequence Player 'StaminaCat_HeadFins_Loop'!(bLoopAnimation:bool=True) -> (Pose) {bLoopAnimation=false, sequence=&StaminaCat_HeadFins_Loop})]
 }
 BigMovement {
  path=AnimGraph.AnimGraphNode_StateMachine_0.Random State Machine.AnimStateNode_1.BigMovement
  nodes [
   n1 = Sequence Player 'StaminaCat_HeadFins_Action'!(bLoopAnimation:bool=True) -> (Pose) {bLoopAnimation=false, sequence=&StaminaCat_HeadFins_Action}
   Output Animation Pose(Result=@n1)
  ]
 }
 Transition {
  path=AnimGraph.AnimGraphNode_StateMachine_0.Random State Machine.AnimStateTransitionNode_4.Transition
  nodes [
   n1 = "Get Relevant Anim Time Remaining (SmallMovement)"!() -> (ReturnValue:float) {is=title K2Node_AnimGetter}
   Result(bCanEnterTransition=AND Boolean(A="float < float"(A=@n1, B=0.1), B=Get PlayBigMovement() -> (PlayBigMovement:bool)))
  ]
 }
 Transition {
  path=AnimGraph.AnimGraphNode_StateMachine_0.Random State Machine.AnimStateTransitionNode_5.Transition
  nodes [
   n1 = Get PlayBigMovement()
   n2 = "Get Relevant Anim Time Remaining (BigMovement)"!() -> (ReturnValue:float) {is=title K2Node_AnimGetter}
   Result(bCanEnterTransition=AND Boolean(A="float < float"(A=@n2, B=0.1), B=NOT Boolean(A=@n1)))
  ]
 }
 EventGraph {nodes [
   AnimNotify_RandomTrigger() {is=title K2Node_Event}
   > Set PlayBigMovement(PlayBigMovement:bool=RandomBool!() -> (ReturnValue:bool)) -> (Output_Get:bool)
   TryGetPawnOwner()
   RandomBoolWithWeight!(Weight:float=0.4) -> (ReturnValue:bool)
   Event BlueprintUpdateAnimation()
  ]}
]
variables [PlayBigMovement {type=boolean}]
