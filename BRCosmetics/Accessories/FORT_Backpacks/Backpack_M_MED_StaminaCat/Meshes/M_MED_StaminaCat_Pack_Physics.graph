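# Physics asset for the StaminaCat backpack cosmetic, defining collision bodies and joints for the pack, a fish, and its water.
# The fish body joins the pack and the head joins the body with free swing and twist.
# Limited joints attach flippers and an "up" bone to the head, and a three-box tail chain to the body, loosening toward the tip.
# A central water sphere hangs off the pack capsule, with tightly limited left and right spheres attached.
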PhysicsAsset /BRCosmetics/Accessories/FORT_Backpacks/Backpack_M_MED_StaminaCat/Meshes/M_MED_StaminaCat_Pack_Physics
bodies {
 dyn_flippers_l [{
   shapeType=Capsule
   length=1.5807
   radius=1.8524
   shapeName=dyn_flippers_l_capsule
   center=<1.7724 -0.654 -2.1963>
   rotation=<rot 12.9587 -94.6721 -41.0515>
  }]
 dyn_flippers_r [{
   shapeType=Capsule
   length=1.5739
   radius=1.7554
   shapeName=dyn_flippers_r_capsule
   center=<0.7417 -1.9382 -1.4978>
   rotation=<rot 27.5932 -129.3366 -53.3353>
  }]
 dyn_tail_01 [{
   shapeType=Box
   extentX=4.1372
   extentY=3.6048
   extentZ=4.1235
   shapeName=dyn_tail_01_box
   center=<0.5315 -0.0258 -2.6149>
   rotation=<rot -0.0001 -88.2417 -103.9727>
  }]
 dyn_tail_02 [{
   shapeType=Box
   extentX=3.1458
   extentY=2.885
   extentZ=4.1455
   shapeName=dyn_tail_02_box
   center=<-0.5394 0.041 -1.4159>
   rotation=<rot 0 -90.2311 -75.1788>
  }]
 dyn_tail_03 [{
   shapeType=Box
   extentX=2.351
   extentY=3.6269
   extentZ=4.9274
   shapeName=dyn_tail_03_box
   center=<-2.1296 -0.0727 -1.8312>
   rotation=<rot 0.0001 91.3423 -46.9159>
  }]
 dyn_up [{shapeType=Capsule, length=0.4293, radius=1.7495, shapeName=dyn_up_capsule, center=<-1.2364 0.0179 1.228>, rotation=<rot 0 -92.4099 -75.4962>}]
 dyn_water_l [{shapeType=Sphere, radius=3.8379, shapeName=dyn_water_l_sphere, center=<0 -0.8628 -1.379>}]
 dyn_water_mid [{shapeType=Sphere, radius=6.0947, shapeName=dyn_water_mid_sphere, center=<0 0.6507 -3.3236>}]
 dyn_water_r [{shapeType=Sphere, radius=3.9276, shapeName=dyn_water_r_sphere, center=<0 0.5634 -1.4385>}]
 fish_body [{
   shapeType=Capsule
   length=1.346
   radius=0.75
   shapeName=fish_body_capsule
   center=<-3.8222 0.6166 5.0886>
   rotation=<rot 5.9259 -90 -34.3386>
  }]
 fish_head [{shapeType=Capsule, length=1.0392, radius=0.5, shapeName=fish_head_capsule, center=<-2.0658 0.4354 6.6137>, rotation=<rot 6.1383 -90 -44.044>}]
 pack [{shapeType=Capsule, length=13.3101, radius=9.312, center=<-2.8262 -7.2022 -6.035>, rotation=<rot -13.9322 -90 -23.6028>}]
}
constraints [
 {bone1Name=dyn_tail_01, bone2Name=fish_body, swing1LimitDegrees=8, swing2LimitDegrees=8, twistLimitDegrees=5}
 {bone1Name=dyn_tail_02, bone2Name=dyn_tail_01, swing1LimitDegrees=10, swing2LimitDegrees=10, twistLimitDegrees=10}
 {bone1Name=dyn_tail_03, bone2Name=dyn_tail_02, swing1LimitDegrees=15, swing2LimitDegrees=15, twistLimitDegrees=10}
 {bone1Name=dyn_water_mid, bone2Name=pack, swing1LimitDegrees=10, swing2LimitDegrees=10, twistLimitDegrees=5}
 {bone1Name=dyn_water_l, bone2Name=dyn_water_mid, swing1LimitDegrees=5, swing2LimitDegrees=10, twistLimitDegrees=2}
 {bone1Name=dyn_water_r, bone2Name=dyn_water_mid, swing1LimitDegrees=5, swing2LimitDegrees=10, twistLimitDegrees=2}
 {bone1Name=dyn_up, bone2Name=fish_head, swing1LimitDegrees=15, swing2LimitDegrees=10, twistLimitDegrees=7}
 {
  bone1Name=fish_body
  bone2Name=pack
  swing1LimitDegrees=45
  swing1Motion=Free
  swing2LimitDegrees=45
  swing2Motion=Free
  twistLimitDegrees=45
  twistMotion=Free
 }
 {
  bone1Name=fish_head
  bone2Name=fish_body
  swing1LimitDegrees=45
  swing1Motion=Free
  swing2LimitDegrees=45
  swing2Motion=Free
  twistLimitDegrees=45
  twistMotion=Free
 }
 {bone1Name=dyn_flippers_l, bone2Name=fish_head, swing1LimitDegrees=10, swing2LimitDegrees=10, twistLimitDegrees=5}
 {bone1Name=dyn_flippers_r, bone2Name=fish_head, swing1LimitDegrees=15, swing2LimitDegrees=15, twistLimitDegrees=5}
]
